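; Blueprint component for previewing Sparks songs in a song player.
; It overrides the StartPreviewInternal event, which takes a catalog data object reference.
; The override only forwards that catalog data to the parent implementation and adds no logic of its own.
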(event EventStartPreviewInternal
  (params (CatalogData :type Catalog%20Data%20Object%20Reference))
  (|CallParentFunction CatalogData))
